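# UI material for the Noble item-shop background, adding an animated radial streak effect over a background texture.
# Radial coordinates are panned and tiled into two samples of a noise-slot test texture, tinted blue and white, then summed.
# A powered radial distance scales the streaks, which a mask texture blends into the background before a second blend softens them.
# A separate vertical falloff branch and an empty Multiply feed nothing shown.
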Material /ItemShop/UI/Art/Noble/M_UI_NobleStoreBG
graph {
 properties {bUsedWithStaticMesh=false, materialDomain=MD_UI, pixelDepthOffsetMode=PDOM_Legacy, usedShadingModels=Default Lit}
 nodes [
  n1 = TextureSampleParameter2D() {parameterName=BG, texture=&/UI_T_NobleStoreBG}
  n2 = MaterialFunctionCall() {materialFunction=&VectorToRadialValue}
  n3 = Panner(
   Coordinate=@n2.Radial Coordinates
   Speed=AppendVector(A=ScalarParameter() {parameterName=U-Speed}, B=ScalarParameter() {defaultValue=-1.1389, parameterName=V-Speed})
  )
  n4 = ScalarParameter() {defaultValue=3.8372, parameterName=R-U-Tiles}
  n5 = ScalarParameter() {defaultValue=0.015, parameterName=R-V-Tiles}
  n6 = Multiply(A=Multiply(A=@n3, B=AppendVector(A=@n4, B=@n5)), B=ScalarParameter() {defaultValue=4.7189, parameterName=TextureSize2})
  n7 = ScalarParameter() {defaultValue=0.4198, parameterName=G-U-Tiles}
  n8 = ScalarParameter() {defaultValue=0.0018, parameterName=G-V-Tiles}
  n9 = Multiply(A=Multiply(A=@n3, B=AppendVector(A=@n7, B=@n8)), B=ScalarParameter() {defaultValue=12.9798, parameterName=TextureSize})
  n10 = Add(
   A=Multiply(A=TextureSample(UVs=@n6) {texture=&/T_UI_Test}.G, B=Constant3Vector() {constant=<rgba 0 0.3464 1 1>}) {constB=30.3558}
   B=Multiply(A=TextureSample(UVs=@n9) {texture=&T_UI_Test}.G, B=Constant3Vector() {constant=<rgba 1 1 1 1>}) {constB=12.2837}
  )
  n11 = Power(Base=@n2.Linear Distance, Exp=ScalarParameter() {defaultValue=0.5449, parameterName=MaskPower})
  LinearInterpolate(
   A=LinearInterpolate(
    A=@n1.RGB
    B=Multiply(A=@n11, B=@n10)
    Alpha=TextureSampleParameter2D() {parameterName=Param, texture=&/UI_T_NobleMask}.RGB
   )
   B=@n1.RGB
  )
  n12 = Constant3Vector() {constant=<rgba 1 1 1 1>}
  n13 = ConstantBiasScale(None=TextureCoordinate()) {bias=-0.5, scale=2}
  n14 = Subtract(A=ComponentMask(None=Abs(None=@n13)) {g=true}, B=ScalarParameter() {defaultValue=0.2601, parameterName=Subtract})
  n15 = ScalarParameter() {defaultValue=-0.1133, parameterName=Falloff}
  n16 = Multiply(A=Saturate(None=Multiply(A=@n14, B=@n15)), B=ScalarParameter() {defaultValue=18.5746, parameterName=Intensity})
  LinearInterpolate(A=LinearInterpolate(B=Multiply(A=@n12, B=@n16)))
  Multiply()
 ]
}
